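; Blueprint for the Brimstone hanging cage A prop, handling material setup for instances streamed in at runtime.
; On BeginPlay it fetches building texture data and, if that array is empty, calls GetMaterialfromRootMesh to generate material instances.
; It then always calls SetDownRodMaterial.
; The construction script only calls its parent, so any further construction logic is inherited.
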(event EventBeginPlay
  ;[1 1 1] Handle generating the MIDs for instances streamed in
  (bind _outtexturedataarray (Building|GetBuildingTextureData))
  (if (not (> (Utilities|Array|Length _outtexturedataarray) 0))
    (CallFunction|GetMaterialfromRootMesh))
  (CallFunction|SetDownRodMaterial))

(fn ConstructionScript
  (|CallParentFunction))
